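; Harvester player augment ability that reacts to a gameplay event and may spawn something, likely a bush, where the target actor was destroyed.
; Activation records the target and continues into parent logic only if it matches the spawnable tags, saving its location first.
; Overrides place spawns at that location with slight horizontal jitter plus a world-space offset.
; The success check fails if the destroyed actor has the queried tags, otherwise deferring to the parent.
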(event Ability|EventActivateAbilityFromEvent
  (params (EventData :type Gameplay%20Event%20Data%20Structure%20%28by%20ref%29))
  (bind (_eventtag _instigator _target) (Utilities|Struct|BreakGameplayEventData EventData))
  ;[1 0.2718 0.06097] Override Activation conditions and set location
  (bind _output_get (Variables|Default|SetDestroyedActor _target))
  (if (CallFunction|CanSpawnBush? _output_get)
    (Variables|Default|SetHit%20Location (Transformation|GetActorLocation _target))
    (|CallParentFunction EventData)))

(fn CanSpawnBush?
  (params (Actor :type Actor%20Object%20Reference))
  (bind _returnvalue (GameplayTags|HasAnyMatchingGameplayTags(Message) Actor (Variables|Default|GetSpawnableTags)))
  (return _returnvalue))

(fn GetLocationToSpawn
  (return (Utilities|Operators|Add (Variables|Default|GetHit%20Location) (Math|Vector|MakeVector (Math|Random|RandomFloatinRange -25.0 25.0) (Math|Random|RandomFloatinRange -25.0 25.0)) (Variables|Default|GetWorldSpaceSpawnOffset))))

(fn GetSpawnSuccess
  (params (Percent :type Float%20%28double-precision%29) (TagToQuery :type Gameplay%20Tag%20Container%20Structure))
  (bind _returnvalue (GameplayTags|HasAnyMatchingGameplayTags(Message) (Variables|Default|GetDestroyedActor) TagToQuery))
  (if _returnvalue
    (return false)
    (else
      (return (|CallParentFunction Percent TagToQuery)))))
